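# Sound cue for the first-person cage saddle bounce, playing a random heavy bounce sample in the Fort_Gadgets sound class.
# A quality-level root sends its first two levels to a pool of eight equally weighted waves and its last two to a reduced pool of three.
# Each pool passes through a modulator that varies pitch slightly downward at fixed volume.
# The cue uses a shared bounce attenuation asset and overrides concurrency to block new instances at the limit.
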SoundCue /MotherGameplay/Sounds/CageSaddle/Cage_Saddle_Bounce_1P_Cue
root=QualityLevel_0
nodes [
 {id=Modulator_0, children [Random_1], properties {pitchMax=1, volumeMax=1, volumeMin=1}}
 {id=Modulator_1, children [Random_2], properties {pitchMax=1, volumeMax=1, volumeMin=1}}
 {id=QualityLevel_0, children [Modulator_0,Modulator_0,Modulator_1,Modulator_1]}
 {
  id=Random_1
  children [WavePlayer_0, WavePlayer_1, WavePlayer_3, WavePlayer_4, WavePlayer_8, WavePlayer_9, WavePlayer_10, WavePlayer_11]
  properties {weights [1,1,1,1,1,1,1,1]}
 }
 {id=Random_2, children [WavePlayer_12, WavePlayer_13, WavePlayer_14], properties {weights [1,1,1]}}
 {id=WavePlayer_0, properties {soundWaveAssetPtr=&/Cage_Saddle_Bounce_Heavy_01}}
 {id=WavePlayer_1, properties {soundWaveAssetPtr=&/Cage_Saddle_Bounce_Heavy_02}}
 {id=WavePlayer_10, properties {soundWaveAssetPtr=&/Cage_Saddle_Bounce_Heavy_07}}
 {id=WavePlayer_11, properties {soundWaveAssetPtr=&/Cage_Saddle_Bounce_Heavy_08}}
 {id=WavePlayer_12, properties {soundWaveAssetPtr=&Cage_Saddle_Bounce_Heavy_01}}
 {id=WavePlayer_13, properties {soundWaveAssetPtr=&Cage_Saddle_Bounce_Heavy_02}}
 {id=WavePlayer_14, properties {soundWaveAssetPtr=&Cage_Saddle_Bounce_Heavy_07}}
 {id=WavePlayer_3, properties {soundWaveAssetPtr=&/Cage_Saddle_Bounce_Heavy_03}}
 {id=WavePlayer_4, properties {soundWaveAssetPtr=&/Cage_Saddle_Bounce_Heavy_04}}
 {id=WavePlayer_8, properties {soundWaveAssetPtr=&/Cage_Saddle_Bounce_Heavy_05}}
 {id=WavePlayer_9, properties {soundWaveAssetPtr=&/Cage_Saddle_Bounce_Heavy_06}}
]
properties {
 attenuationSettings=&/Cage_Saddle_Bounce_Att
 bOverrideConcurrency=true
 maxDistance=2400
 soundClassObject=&Fort_Gadgets
 volumeMultiplier=0.7
 attenuationOverrides {falloffDistance=1000, attenuationShapeExtents {x=500}}
 concurrencyOverrides {resolutionRule=PreventNew}
}
