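# Sound cue for the Cage Saddle inflate sound, assigned to the Fort_Gadgets sound class.
# A Random node picks one of four equally weighted inflate waves and feeds a Modulator at the root for variation.
# It uses the Cage_Saddle_Bounce_Att attenuation asset, and its concurrency override prevents new instances from starting once the limit is reached.
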SoundCue /MotherGameplay/Sounds/CageSaddle/Cage_Saddle_Inflate_Cue
root=Modulator_2
nodes [
 {id=Modulator_2, children [Random_1]}
 {id=Random_1, children [WavePlayer_0,WavePlayer_1,WavePlayer_3,WavePlayer_4], properties {weights [1,1,1,1]}}
 {id=WavePlayer_0, properties {soundWaveAssetPtr=&/Cage_Saddle_Inflate_01}}
 {id=WavePlayer_1, properties {soundWaveAssetPtr=&/Cage_Saddle_Inflate_02}}
 {id=WavePlayer_3, properties {soundWaveAssetPtr=&/Cage_Saddle_Inflate_03}}
 {id=WavePlayer_4, properties {soundWaveAssetPtr=&/Cage_Saddle_Inflate_04}}
]
properties {
 attenuationSettings=&/Cage_Saddle_Bounce_Att
 bOverrideConcurrency=true
 maxDistance=2400
 soundClassObject=&Fort_Gadgets
 attenuationOverrides {falloffDistance=1000, attenuationShapeExtents {x=500}}
 concurrencyOverrides {resolutionRule=PreventNew}
}
