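# Sound cue for the first-person Cage Saddle jump sound, routed through the Fort_Gadgets sound class.
# A Random node picks one of four equally weighted jump wave variations and feeds a root Modulator node.
# It uses the Cage_Saddle_Bounce attenuation asset and overrides concurrency so new instances are blocked rather than stopping playing ones.
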SoundCue /MotherGameplay/Sounds/CageSaddle/Cage_Saddle_Jump_1P_Cue
root=Modulator_2
nodes [
 {id=Modulator_2, children [Random_0]}
 {id=Random_0, children [WavePlayer_2,WavePlayer_5,WavePlayer_6,WavePlayer_7], properties {weights [1,1,1,1]}}
 {id=WavePlayer_2, properties {soundWaveAssetPtr=&/Cage_Saddle_Jump_01}}
 {id=WavePlayer_5, properties {soundWaveAssetPtr=&/Cage_Saddle_Jump_02}}
 {id=WavePlayer_6, properties {soundWaveAssetPtr=&/Cage_Saddle_Jump_03}}
 {id=WavePlayer_7, properties {soundWaveAssetPtr=&/Cage_Saddle_Jump_04}}
]
properties {
 attenuationSettings=&/Cage_Saddle_Bounce_Att
 bOverrideConcurrency=true
 maxDistance=2400
 soundClassObject=&Fort_Gadgets
 attenuationOverrides {falloffDistance=1000, attenuationShapeExtents {x=500}}
 concurrencyOverrides {resolutionRule=PreventNew}
}
