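# Looping movement sound cue for the Cage Saddle, classed under gadget sounds, with distance attenuation and a prevent-new concurrency rule.
# A mixer blends three loops through continuous modulators: ground roll follows OnGround, borrowed skydive wind follows its inverse, and water follows InWater.
# Roll is boosted and wind is kept low in the mix.
# A root modulator scales the whole mix's volume and pitch by the Speed parameter.
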SoundCue /MotherGameplay/Sounds/CageSaddle/Cage_Saddle_Movement_Cue
root=ModulatorContinuous_2
nodes [
 {id=Mixer_0, children [ModulatorContinuous_1, ModulatorContinuous_3, ModulatorContinuous_0], properties {inputVolume [2,0.25,1]}}
 {
  id=ModulatorContinuous_0
  children [WavePlayer_0]
  properties {pitchModulationParams {default=0, minOutput=1, parameterName=InWater}, volumeModulationParams {default=0, parameterName=InWater}}
 }
 {
  id=ModulatorContinuous_1
  children [WavePlayer_2]
  properties {pitchModulationParams {default=0, minOutput=1, parameterName=OnGround}, volumeModulationParams {default=0, parameterName=OnGround}}
 }
 {
  id=ModulatorContinuous_2
  children [Mixer_0]
  properties {pitchModulationParams {default=0, minOutput=0.7, parameterName=Speed}, volumeModulationParams {default=0, parameterName=Speed}}
 }
 {
  id=ModulatorContinuous_3
  children [WavePlayer_5]
  properties {
   pitchModulationParams {default=0, minOutput=1, parameterName=OnGround}
   volumeModulationParams {default=0, maxOutput=0, minOutput=1, parameterName=OnGround}
  }
 }
 {id=WavePlayer_0, properties {bLooping=true, soundWaveAssetPtr=&/Cage_Saddle_Water_Loop_01}}
 {id=WavePlayer_2, properties {bLooping=true, soundWaveAssetPtr=&/Cage_Saddle_Roll_Loop_01}}
 {id=WavePlayer_5, properties {bLooping=true, soundWaveAssetPtr=/Game/Athena/Sounds/Wind_Falling/AMB_Wind_Skydive_Horizontal_Loop_01}}
]
properties {
 attenuationSettings=&/Cage_Saddle_Bounce_Att
 bOverrideConcurrency=true
 maxDistance=2400
 soundClassObject=&Fort_Gadgets
 volumeMultiplier=1
 attenuationOverrides {falloffDistance=1000, attenuationShapeExtents {x=500}}
 concurrencyOverrides {resolutionRule=PreventNew}
}
