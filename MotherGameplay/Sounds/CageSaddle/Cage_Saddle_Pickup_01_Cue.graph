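# Sound cue for the Cage Saddle pickup sound, apparently played when that item is collected as loot.
# A single wave player feeds the Cage_Saddle_Pickup_01 wave straight to the output, with no randomization or mixing nodes.
# It uses the shared weapon pickup attenuation and loot collect sound class, with reduced volume.
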SoundCue /MotherGameplay/Sounds/CageSaddle/Cage_Saddle_Pickup_01_Cue
root=WavePlayer_0
nodes [{id=WavePlayer_0, properties {soundWaveAssetPtr=&/Cage_Saddle_Pickup_01}}]
properties {
 attenuationSettings=&Fort_Weapon_Pickup_Att
 maxDistance=3500
 soundClassObject=&Fort_Loot_Collect
 volumeMultiplier=0.398
}
