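# Sound cue for the Cage Saddle "ready" sound, apparently signalling that its timer is ready.
# A single timer-ready wave plays through a Modulator at its default settings, at reduced volume in the Fort_Gadgets sound class.
# It references the Cage Saddle bounce attenuation asset and overrides concurrency so new instances are prevented once the limit is reached.
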SoundCue /MotherGameplay/Sounds/CageSaddle/Cage_Saddle_Ready_Cue
root=Modulator_2
nodes [
 {id=Modulator_2, children [WavePlayer_1]}
 {id=WavePlayer_1, properties {soundWaveAssetPtr=&/Cage_Saddle_Timer_Ready_01}}
]
properties {
 attenuationSettings=&/Cage_Saddle_Bounce_Att
 bOverrideConcurrency=true
 maxDistance=2400
 soundClassObject=&Fort_Gadgets
 volumeMultiplier=0.35
 attenuationOverrides {falloffDistance=1000, attenuationShapeExtents {x=500}}
 concurrencyOverrides {resolutionRule=PreventNew}
}
