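# Sound Cue for the Cage Saddle timer tick, routed through the Fort_Gadgets sound class.
# A single looping Wave Player plays the Cage_Saddle_Timer_Tick_02 wave at reduced volume.
# Inline attenuation overrides keep it short-range, taking the place of the assigned Cage_Saddle_Bounce_Att asset.
# An inline concurrency override rejects new instances once the limit is reached.
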SoundCue /MotherGameplay/Sounds/CageSaddle/Cage_Saddle_Timer_Tick_Cue
root=WavePlayer_2
nodes [{id=WavePlayer_2, properties {bLooping=true, soundWaveAssetPtr=&/Cage_Saddle_Timer_Tick_02}}]
properties {
 attenuationSettings=&/Cage_Saddle_Bounce_Att
 bOverrideAttenuation=true
 bOverrideConcurrency=true
 maxDistance=600
 soundClassObject=&Fort_Gadgets
 volumeMultiplier=0.3
 attenuationOverrides {falloffDistance=500, attenuationShapeExtents {x=100}}
 concurrencyOverrides {resolutionRule=PreventNew}
}
