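# Sound cue for the water exit splash of the Cage Saddle, routed to the Fort_Gadgets sound class.
# A Random node picks one of four splash exit wave variations with equal weight.
# The cue plays at reduced volume and is spatialized through the shared Cage_Saddle_Bounce_Att attenuation asset.
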SoundCue /MotherGameplay/Sounds/CageSaddle/Cage_Saddle_Water_Splash_Exit_Cue
root=Random_0
nodes [
 {id=Random_0, children [WavePlayer_0,WavePlayer_1,WavePlayer_2,WavePlayer_3], properties {weights [1,1,1,1]}}
 {id=WavePlayer_0, properties {soundWaveAssetPtr=&/Cage_Saddle_Water_Splash_Exit_01}}
 {id=WavePlayer_1, properties {soundWaveAssetPtr=&/Cage_Saddle_Water_Splash_Exit_02}}
 {id=WavePlayer_2, properties {soundWaveAssetPtr=&/Cage_Saddle_Water_Splash_Exit_03}}
 {id=WavePlayer_3, properties {soundWaveAssetPtr=&/Cage_Saddle_Water_Splash_Exit_04}}
]
properties {
 attenuationSettings=&/Cage_Saddle_Bounce_Att
 maxDistance=2400
 soundClassObject=&Fort_Gadgets
 volumeMultiplier=0.4
}
